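# Sound cue for the fully-charged reactive effect on the female Parfait Commando skin, played as player foley.
# A random node picks one of three equally weighted Fullcharge waves, then a modulator adds slight volume variation.
# It uses basic foley attenuation and overrides concurrency to a single instance, stopping the oldest when retriggered.
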SoundCue /Game/Athena/Sounds/Character/Skin/Commando_Parfait_Female/Char_ParfaitPowers_Reactive_Cue
root=Modulator_0
nodes [
 {id=Modulator_0, children [Random_0], properties {pitchMax=1, volumeMax=1.1, volumeMin=1}}
 {id=Random_0, children [WavePlayer_0,WavePlayer_1,WavePlayer_2], properties {weights [1,1,1]}}
 {
  id=WavePlayer_0
  properties {soundWaveAssetPtr=&/Skin_Commando_ParfaitFemale_Fullcharge_01}
 }
 {
  id=WavePlayer_1
  properties {soundWaveAssetPtr=&/Skin_Commando_ParfaitFemale_Fullcharge_02}
 }
 {
  id=WavePlayer_2
  properties {soundWaveAssetPtr=&/Skin_Commando_ParfaitFemale_Fullcharge_03}
 }
]
properties {
 attenuationSettings=&Foley_Attenuation_Basic_Att
 bOverrideConcurrency=true
 maxDistance=1536
 soundClassObject=&Fort_Player_Foley
 volumeMultiplier=1
 concurrencyOverrides {maxCount=1, resolutionRule=StopOldest}
}
